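; Boss egg gameplay actor that drops to the ground once all its registered connections are broken, then spawns a boss placeholder actor.
; The construction script traces downward to store the impact point and aims and sizes an arrow toward it.
; Authority-only events count established and lost connections, triggering Fall when losses reach the total.
; Fall lerps the egg down via timeline, plays an explosion, hides it, spawns the placeholder and destroys itself after a delay.
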(event Custom|Fall
  (Variables|Default|SetStartingLocation (Transformation|GetActorLocation))
  (bind (_direction _progress) (Timeline|Falling PlayFromStart)
    (:Update
      (Transformation|SetActorLocation (Math|Vector|Lerp(Vector) (Variables|Default|GetStartingLocation) (Variables|Default|GetIntendedImpactPoint) _progress)))
    (:Finished
      (Niagara|SpawnSystematLocation :SystemTemplate "/Game/Weapons/FORT_GrenadeLaunchers/Effects/NS_GrenadeLauncher_Explosion_Athena_Pumpkin.NS_GrenadeLauncher_Explosion_Athena_Pumpkin" :Location (Transformation|GetActorLocation) :Scale "10.000000,10.000000,10.000000" :bAutoDestroy true :bAutoActivate true :PoolingMethod "None" :bPreCullCheck true)
      (Rendering|SetActorHiddenInGame true)
      (Building|SpawnBuildingGameplayActor "/ProtoSwarm/Blueprints/EnvironmentalGameplay/BGA_BossPlaceholder.BGA_BossPlaceholder_C" (Math|Transform|MakeTransform (Variables|Default|GetIntendedImpactPoint)) self true false)
      (Utilities|FlowControl|Delay 0.6)
      (Actor|DestroyActor))))

(event Custom|BreakConnection
  (Networking|SwitchHasAuthority
    (:Authority
      (Math|Integer|IncrementInt (Variables|Default|GetNumConnectionsLost)
        (:%20%20
          (Development|PrintString "Replicated connections lost" true true "(R=1.000000,G=0.000000,B=0.071892,A=1.000000)" 1e+01)
          (if (> (Variables|Default|GetNumConnections) 0)
            (if (>= (Variables|Default|GetNumConnectionsLost) (Variables|Default|GetNumConnections))
              (CallFunction|Fall))))))))

(event Collision|EventActorBeginOverlap
  (params (OtherActor :type Actor%20Object%20Reference)))

(event Custom|EstablishConnection
  (Networking|SwitchHasAuthority
    (:Authority
      (Math|Integer|IncrementInt (Variables|Default|GetNumConnections)))))

(event EventTick
  (params (DeltaSeconds :type Float%20%28single-precision%29)))

(fn ConstructionScript
  (bind _returnvalue (Transformation|GetActorLocation))
  (bind _arrow (Variables|Default|GetArrow))
  (|CallParentFunction)
  (bind (_outhit _returnvalue_2) (Collision|LineTraceByChannel (Transformation|GetActorLocation) (- (Transformation|GetActorLocation) "0.000000,0.000000,10000.000000") "TraceTypeQuery10"))
  (if _returnvalue_2
    (Variables|Default|SetIntendedImpactPoint (.ImpactPoint (Collision|BreakHitResult _outhit)))
    (Variables|/Script/Engine.ArrowComponent|SetArrowLength (/ (Math|Vector|VectorLength (- (.ImpactPoint (Collision|BreakHitResult _outhit)) _returnvalue)) (Variables|/Script/Engine.ArrowComponent|GetArrowSize (Variables|Default|GetArrow))) _arrow)
    (Transformation|SetWorldRotation _arrow (Math|Conversions|RotationFromXVector (- (.ImpactPoint (Collision|BreakHitResult _outhit)) _returnvalue)) false false)))
