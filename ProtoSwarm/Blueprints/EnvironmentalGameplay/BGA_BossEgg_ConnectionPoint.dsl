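; Boss egg connection point actor that links itself to an egg on BeginPlay and breaks that link once all its registered weakspots are destroyed.
; The construction script positions, aims and uniformly scales a connection component to span from the Z-offset attachment point to above the egg.
; Weakspot counters trigger the final break: an egg cam call (empty here), a short delay, then a multicast that spawns explosions up the attachment height and destroys the actor.
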(event EventBeginPlay
  (|CallParentFunction)
  (Class|BGABossEgg|EstablishConnection (Variables|Default|GetTheEgg)))

(event Collision|EventActorBeginOverlap
  (params (OtherActor :type Actor%20Object%20Reference)))

(event EventTick
  (params (DeltaSeconds :type Float%20%28single-precision%29)))

(event Custom|RegisterWeakspot
  (Math|Integer|IncrementInt (Variables|Default|GetWeakspotNum)))

(event Custom|RegisterWeakspotDestroyed
  (params (Damage%20Instigator :type Pawn%20Object%20Reference))
  (bind _result (Math|Integer|IncrementInt (Variables|Default|GetWeakspotsDestroyed))
    (:%20%20
      (if (>= _result (Variables|Default|GetWeakspotNum))
        (CallFunction|ShowEggCam Damage%20Instigator)
        (Utilities|FlowControl|Delay 0.6)
        (Class|BGABossEgg|BreakConnection (Variables|Default|GetTheEgg))
        (CallFunction|DestroyandSpawnEffectsMulticast)))))

(event Custom|DestroyAndSpawnEffects_Multicast
  (for _index (range (+ (Math|Float|Truncate (/ (Variables|Default|GetAttachmentZOffset) 4e+02)) 1))
    (Niagara|SpawnSystematLocation :SystemTemplate "/Game/Weapons/FORT_GrenadeLaunchers/Effects/NS_GrenadeLauncher_Explosion_Athena_Pumpkin.NS_GrenadeLauncher_Explosion_Athena_Pumpkin" :Location (Utilities|Operators|Add (Math|Vector|MakeVector 0.0 0.0 (* _index 400)) (Transformation|GetWorldLocation (Variables|Default|GetStaticMesh))) :Scale "1.500000,1.500000,1.500000" :bAutoDestroy true :bAutoActivate true :PoolingMethod "None" :bPreCullCheck true))
  (Actor|DestroyActor))

(event Custom|ShowEggCam
  (params (Player%20Pawn :type Pawn%20Object%20Reference)))

(fn ConstructionScript
  (bind _theegg (Variables|Default|GetTheEgg))
  (bind _returnvalue (Transformation|GetActorLocation _theegg))
  (bind _returnvalue_2 (Transformation|GetActorLocation))
  (bind _attachmentzoffset (Variables|Default|GetAttachmentZOffset))
  (bind _returnvalue_3 (Utilities|Operators|Add _returnvalue_2 (Math|Vector|MakeVector 0.0 0.0 _attachmentzoffset)))
  (bind _returnvalue_4 (+ _returnvalue "0.000000,0.000000,600.000000"))
  (bind _returnvalue_5 (- _returnvalue_4 _returnvalue_3))
  (bind _returnvalue_6 (Math|Vector|VectorLength _returnvalue_5))
  (bind _returnvalue_7 (* _returnvalue_6 0.0019))
  (Utilities|IsValid _theegg
    (:Is%20Valid
      (Transformation|SetWorldLocationAndRotation (Variables|Default|GetConnection) (Utilities|Operators|Add (* (* _returnvalue_7 -5e+01) (Math|Vector|GetUpVector (Math|Conversions|RotationFromXVector (Math|Vector|Normalize _returnvalue_5)))) (* _returnvalue_5 0.5) _returnvalue_3) (Math|Conversions|RotationFromXVector (Math|Vector|Normalize _returnvalue_5)) false true)
      (Transformation|SetWorldScale3D (Variables|Default|GetConnection) (Math|Conversions|ToVector(Float) _returnvalue_7)))))
