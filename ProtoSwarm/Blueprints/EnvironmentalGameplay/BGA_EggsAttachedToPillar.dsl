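; Breakable pillar-egg Building Game Actor that triggers an insect AI spawn through the Living World Manager when damaged.
; On non-cosmetic instances, it waits for playlist data, then reserves fractional AI density, logging any failure.
; Server damage enables the child static point provider, releases the reservation and requests event generation next tick for faster spawning.
; Damage also fires OnSpawnedEnemy and calls DestroySelfMulticast to destroy and hide the actor.
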(event EventBeginPlay
  (Utilities|FlowControl|Sequence
    (:then_1
      (Game|BindEventtoOnDestroyed (delegate Custom|Destroyed))
      (Networking|CanExecuteCosmeticEvents
        (:False
          ; When we spawn Something Dynamically, we must always Register to the Living World Manager that this Assets will be spawning an AI. This must ALWAYS be done after a PlaylistDataReadyAsync, else the Living World Manager might not be ready. In our case, we set the Density to 0.2 for now, because I set that the Insects are worth 0.2 AI Count. It needs to match to make sure we don't take up more budget. If we were spawning 2 AI, I would set the Density in this case to 0.4. Normal NPCs take 1 AI count usually, so if we were spawning an NPC, the Density value would be 1 instead of 0.2 - This number will most probably be changed in the future! The Range Represents the Pawn's NetCull Distance - It needs to be equal or grater than the Pawn's Relevancy Distance.
          (Class|FortAsyncActionCurrentPlaylistReady|PlaylistDataReadyAsync
            (:OnReady
              (Utilities|FlowControl|DelayUntilNextTick)
              (bind _returnvalue (Class|FortAthenaLivingWorldManager|RegisterDensityReservoir (Class|FortAthenaLivingWorldManager|GetLivingWorldManager) self 0.2 1.2e+04))
              (if (not _returnvalue)
                (Fortnite|FortShippingLog self ":: Tried to Reserver 1 AI Spot, but failed. It's highly possible the Budget was just too full or the Living World Manager was not ready. AI can still spawn if at the time of Spawn Request there is budget, but we cannot guarantee it will spawn." true)))))))))

(event Custom|Destroyed
  (params (DestroyedActor :type Actor%20Object%20Reference))
  (Development|PrintString "Hello"))

(event EventOnDamageServer
  (params (Damage :type Float%20%28single-precision%29) (DamageTags :type Gameplay%20Tag%20Container%20Structure%20%28by%20ref%29) (Momentum :type Vector :default "0, 0, 0") (HitInfo :type Hit%20Result%20Structure%20%28by%20ref%29) (InstigatedBy :type Controller%20Object%20Reference) (DamageCauser :type Actor%20Object%20Reference) (EffectContext :type Gameplay%20Effect%20Context%20Handle%20Structure))
  (|CallParentFunction Damage DamageTags Momentum HitInfo InstigatedBy DamageCauser EffectContext)
  (CallFunction|SpawnEnemy)
  (Default|CallOnSpawnedEnemy)
  (CallFunction|DestroySelfMulticast))

(event Custom|SpawnEnemy
  (bind _asfort_athena_living_world_static_point_provider (Utilities|Casting|CastToFortAthenaLivingWorldStaticPointProvider (Variables|/Script/Engine.ChildActorComponent|GetChildActor (Variables|Default|GetAISpawner_ChildActor))))
  ; This will make the Provider Enabled, allowing the Living World Manager to try to spawn AN AI on it, as long as in it's Definition it's allowed to spawn an AI with the matching Tags found on the Actor. That being said, this doesn't guarantee an imediate spawn, the Living World Manager tries to spawn AI every 5 seconds. So it might take up to 5 seconds before your AI actually spawns.
  (Class|FortAthenaLivingWorldPointProviderInterface|EnablePointProvider_DEPRECATED _asfort_athena_living_world_static_point_provider)
  ; Since we enabled our Provider, we need to "Free Up" the AI Budget we reserved when this BGA spawned to make room for the AI we do want to spawn from here. SO now we Unregisted the AI.
  (Class|FortAthenaLivingWorldManager|UnregisterDensityReservoir (Class|FortAthenaLivingWorldManager|GetLivingWorldManager) self)
  ; Now that we have enabled our Provider and Freed up the AI Budget it was using, we want the Living World Manager to spawn this AI as quickly as we can. So instead of waiting for up to 5 seconds to try to repsawn AI, we are telling the Living World Manager to try to spawn them NOW instead.
  (Utilities|FlowControl|DelayUntilNextTick)
  (Class|FortAthenaLivingWorldManager|RequestEventGeneration (Class|FortAthenaLivingWorldManager|GetLivingWorldManager))
  (Rendering|SetActorHiddenInGame true)
  (Actor|DestroyActor))

(event Custom|DestroySelfMulticast
  (Actor|DestroyActor)
  (Rendering|SetActorHiddenInGame true))

(fn ConstructionScript)
