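; Angled weakspot actor for Swarm boss egg pillars that reports to its pillar and releases Tidal Crane enemies when destroyed.
; Construction finds a nearby BGA_BossEgg_ConnectionPoint by sphere overlap.
; On BeginPlay, the server spawns a target prop, binds its death event and registers with that pillar.
; Target death triggers a multicast that hides and destroys this actor, reports the instigator's pawn to the pillar and spawns cranes, partly randomized.
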(event EventBeginPlay
  (Networking|SwitchHasAuthority
    (:Authority
      (bind _returnvalue (Game|SpawnActorfromClass "/ProtoSwarm/Blueprints/EnvironmentalGameplay/Prop_WeakspotTarget_Swarm.Prop_WeakspotTarget_Swarm_C" (Transformation|GetWorldTransform (Variables|Default|GetScene)) "Undefined" "MultiplyWithRoot" 0 0 0 0))
      (EventDispatchers|BindEventtoOnDied _returnvalue (delegate Custom|OnWeakspotCenterDied) :!class /Script/FortniteGame.BuildingActor)
      (Class|BGABossEggConnectionPoint|RegisterWeakspot (Variables|Default|GetPillarThisIsAttachedTo)))))

(event Collision|EventActorBeginOverlap
  (params (OtherActor :type Actor%20Object%20Reference)))

(event Custom|OnWeakspotCenterDied
  (params (DamagedActor :type Actor%20Object%20Reference) (Damage :type Float%20%28single-precision%29) (InstigatedBy :type Controller%20Object%20Reference) (DamageCauser :type Actor%20Object%20Reference) (HitLocation :type Vector) (FHitComponent :type Primitive%20Component%20Object%20Reference) (BoneName :type Name) (Momentum :type Vector))
  (bind _returnvalue (Transformation|GetActorLocation))
  (bind _returnvalue_2 (+ _returnvalue "0.000000,150.000000,0.000000"))
  (Networking|SwitchHasAuthority
    (:Authority
      (CallFunction|OnWeakspotHitMulticast)
      (Class|BGABossEggConnectionPoint|RegisterWeakspotDestroyed (Variables|Default|GetPillarThisIsAttachedTo) (Pawn|GetControlledPawn InstigatedBy))
      (Networking|SwitchHasAuthority
        (:Authority
          (switch int (Math|Random|RandomIntegerinRange 0 11)
            (:0
              _)
            (:1
              (Game|SpawnActorfromClass "/ProtoSwarm/NPCs/Swarm/TidalCrane_Swarm_Base/Pawns/Pawn_TidalCrane_Swarm_Base.Pawn_TidalCrane_Swarm_Base_C" (Math|Transform|MakeTransform (Transformation|GetActorLocation)) "Undefined" "MultiplyWithRoot" 0 "None" "FAPG_Default" 0 :!label _label))
            (:2
              (goto _label))
            (:3
              _))
          (Game|SpawnActorfromClass "/ProtoSwarm/NPCs/Swarm/TidalCrane_Swarm_Base/Pawns/Pawn_TidalCrane_Swarm_Base.Pawn_TidalCrane_Swarm_Base_C" (Math|Transform|MakeTransform _returnvalue) "Undefined" "MultiplyWithRoot" 0 "None" "FAPG_Default" 0)
          (Game|SpawnActorfromClass "/ProtoSwarm/NPCs/Swarm/TidalCrane_Swarm_Base/Pawns/Pawn_TidalCrane_Swarm_Base.Pawn_TidalCrane_Swarm_Base_C" (Math|Transform|MakeTransform _returnvalue_2) "Undefined" "MultiplyWithRoot" 0 "None" "FAPG_Default" 0)
          (Game|SpawnActorfromClass "/ProtoSwarm/NPCs/Swarm/TidalCrane_Swarm_Base/Pawns/Pawn_TidalCrane_Swarm_Base.Pawn_TidalCrane_Swarm_Base_C" (Math|Transform|MakeTransform (+ _returnvalue_2 "150.000000,-150.000000,0.000000")) "Undefined" "MultiplyWithRoot" 0 "None" "FAPG_Default" 0))))))

(event EventTick
  (params (DeltaSeconds :type Float%20%28single-precision%29)))

(event Custom|OnWeakspotHit_Multicast
  (Collision|SetActorEnableCollision false)
  (Rendering|SetActorHiddenInGame true)
  (Actor|DestroyActor))

(fn ConstructionScript
  (bind (_outactors _returnvalue) (Collision|SphereOverlapActors (Transformation|GetWorldLocation (Variables|Default|GetScene)) 5e+02 (Utilities|Array|MakeArray "ObjectTypeQuery1" "ObjectTypeQuery2") "/ProtoSwarm/Blueprints/EnvironmentalGameplay/BGA_BossEgg_ConnectionPoint.BGA_BossEgg_ConnectionPoint_C"))
  (if _returnvalue
    (bind _asbga_boss_egg_connection_point (Utilities|Casting|CastToBGA_BossEgg_ConnectionPoint (Utilities|Array|Get(acopy) _outactors)))
    (Variables|Default|SetPillarThisIsAttachedTo _asbga_boss_egg_connection_point)))
