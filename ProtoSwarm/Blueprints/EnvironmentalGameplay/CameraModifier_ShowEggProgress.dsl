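; Camera modifier that briefly turns the player's view toward a boss egg, apparently to show its progress.
; Init stores the egg, waits 1.5 seconds, then disables the modifier immediately and removes it from the owning player camera manager.
; BlueprintModifyCamera blends view rotation toward the egg along the shortest path using the modifier's alpha, leaving location and FOV unchanged.
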(event Custom|Init
  (params (TheEgg :type BGA%20Boss%20Egg%20Object%20Reference))
  (Variables|Default|SetThe%20Egg TheEgg)
  (Utilities|FlowControl|Delay 1.5)
  (CameraModifier|DisableModifier true)
  (Game|Player|RemoveCameraModifier (Variables|CameraModifier|GetCameraOwner) self))

(fn BlueprintModifyCamera
  (params (DeltaTime :type Float%20%28single-precision%29) (ViewLocation :type Vector :default "0, 0, 0") (ViewRotation :type Rotator :default "0, 0, 0") (FOV :type Float%20%28single-precision%29))
  (return ViewLocation (Math|Rotator|Lerp(Rotator) ViewRotation (Math|Conversions|RotationFromXVector (- (Transformation|GetActorLocation (Variables|Default|GetThe%20Egg)) ViewLocation)) (Variables|CameraModifier|GetAlpha) true) FOV))
