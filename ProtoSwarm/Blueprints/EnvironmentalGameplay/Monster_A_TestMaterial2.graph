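# Default Lit test material for a monster, with a green default tint, animated glow and wobble.
# Two world-aligned Bubbles normal textures at different scales are multiplied and fed with the tint into MultiplyAdd.
# A scalar-scaled wobble blends two SimpleGrassWind offsets by a time-driven sine.
# The glow multiplies world-aligned foam swirls, panning caustics, GlowEffect, tint and a strong Fresnel ramp that a static switch can flip.
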Material /ProtoSwarm/Blueprints/EnvironmentalGameplay/Monster_A_TestMaterial2
graph {
 properties {bAutomaticallySetUsageInEditor=false, usedShadingModels=Default Lit}
 nodes [
  n1 = TextureObject() {samplerType=SAMPLERTYPE_Normal, texture=/Game/Items/Art_noLOD/Textures/Bubbles}
  n2 = Constant3Vector() {constant=<rgba 512 512 512 0>}
  n3 = TextureObject() {samplerType=SAMPLERTYPE_Normal, texture=&Bubbles}
  n4 = Constant3Vector() {constant=<rgba 1350 1500 1350 0>}
  MaterialFunctionCall!(
   Base=VectorParameter() {parameterName=Param_1, defaultValue=<rgba 0 0.25 0.0209 1>}.RGB
   Add=Multiply(
    A=MaterialFunctionCall(TextureObject=@n1, TextureSize=@n2) {materialFunction=&WorldAlignedTexture}.Z Texture
    B=MaterialFunctionCall(TextureObject=@n3, TextureSize=@n4) {materialFunction=&WorldAlignedTexture}.Z Texture
   )
  ) {materialFunction=/Engine/Functions/Engine_MaterialFunctions02/Math/MultiplyAdd}
  n5 = Constant() {r=0.5}
  n6 = Constant() {r=2}
  n7 = Constant() {r=0.25}
  n8 = Constant()
  n9 = Constant() {r=3}
  n10 = Constant() {r=0.15}
  n11 = Constant() {r=1}
  n12 = Divide(A=Time()) {constB=3}
  Multiply(
   A=ScalarParameter() {defaultValue=0.15, parameterName=Wobble}
   B=LinearInterpolate(
    A=MaterialFunctionCall!(WindIntensity=@n5, WindWeight=@n6, WindSpeed=@n7, AdditionalWPO=@n8) {materialFunction=/Engine/Functions/Engine_MaterialFunctions01/WorldPositionOffset/SimpleGrassWind}.Result
    B=MaterialFunctionCall(WindIntensity=@n9, WindWeight=@n10, WindSpeed=@n11, AdditionalWPO=@n11) {materialFunction=&SimpleGrassWind}.Result
    Alpha=Sine(None=@n12)
   )
  )
  n13 = MaterialFunctionCall(
   TextureObject=TextureObject() {samplerType=SAMPLERTYPE_LinearColor, texture=/Game/Global/Textures/Default/Caustics/03_WaterFoam_AccentSwirls_H}
   TextureSize=Constant3Vector() {constant=<rgba 500 500 500 0>}
  ) {materialFunction=&WorldAlignedTexture}
  n14 = TextureSample(UVs=Panner() {speedX=0.05, speedY=0.01}) {texture=/Game/Global/Textures/Default/Caustics/T_Caustics_Softer}
  n15 = LinearInterpolate(Alpha=Fresnel() {exponent=0.5}) {constA=50, constB=-50}
  n16 = LinearInterpolate(Alpha=Fresnel() {exponent=0.5}) {constA=-50, constB=50}
  Multiply(
   A=Multiply(A=Multiply(A=@n13.XYZ Texture, B=@n14.RGB), B=ScalarParameter() {defaultValue=1, parameterName=GlowEffect})
   B=Multiply(
    A=VectorParameter() {parameterName=Param_1, defaultValue=<rgba 0 0.25 0.0209 1>}.RGB
    B=StaticSwitchParameter(True=@n15, False=@n16) {defaultValue=true, parameterName=FlipFresnel}
   )
  )
  ScalarParameter() {defaultValue=0.1, parameterName=Anis}
 ]
}
